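# Sound cue for dropping a medium cash wad, classed as loot-collect audio with weapon-drop attenuation and wad-pickup concurrency.
# After a short fixed delay at the root, it layers two cash drop waves, the second staggered slightly, under a subtle random pitch modulator.
# A slightly delayed, quieter fanfare wave is mixed in outside the modulator, so its pitch stays constant.
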SoundCue /Game/Athena/Sounds/Items/Wads/Wad_Drop_Med_Cue
root=Delay_0
nodes [
 {id=Delay_0, children [Mixer_1], properties {delayMax=0.2, delayMin=0.2}}
 {id=Delay_1, children [WavePlayer_0], properties {delayMax=0.1, delayMin=0.1}}
 {id=Delay_2, children [WavePlayer_1], properties {delayMax=0.05, delayMin=0.05}}
 {id=Mixer_0, children [WavePlayer_3,Delay_1], properties {inputVolume [1,1]}}
 {id=Mixer_1, children [Modulator_0,Delay_2], properties {inputVolume [1,0.8]}}
 {id=Modulator_0, children [Mixer_0], properties {pitchMax=1.02, pitchMin=0.98}}
 {id=WavePlayer_0, properties {soundWaveAssetPtr=&/Cash_Drop_Med_02}}
 {id=WavePlayer_1, properties {soundWaveAssetPtr=&/Cash_Drop_Med_Fanfare}}
 {id=WavePlayer_3, properties {soundWaveAssetPtr=&/Cash_Drop_Med}}
]
properties {
 attenuationSettings=&Fort_Weapon_Drop_Att
 maxDistance=3000
 priority=0.9
 soundClassObject=&Fort_Loot_Collect
 volumeMultiplier=0.2
 attenuationOverrides {falloffDistance=1000, stereoSpread=512, attenuationShapeExtents {x=0}}
 concurrencyOverrides {maxCount=1, resolutionRule=PreventNew}
 concurrencySet [&/CONC_Wad_Pickup]
}
