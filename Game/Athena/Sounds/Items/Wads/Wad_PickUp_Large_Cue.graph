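# Sound cue for picking up a large wad of cash, played in the loot collect sound class.
# It layers a modulated high cash pickup sample at slightly reduced level with a high pickup tone delayed by a quarter second.
# Weapon pickup attenuation spatializes it, and a wad pickup concurrency set is assigned alongside stop-oldest override values.
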SoundCue /Game/Athena/Sounds/Items/Wads/Wad_PickUp_Large_Cue
root=Mixer_0
nodes [
 {id=Delay_0, children [Mixer_1], properties {delayMax=0.25, delayMin=0.25}}
 {id=Mixer_0, children [Modulator_1,Delay_0], properties {inputVolume [0.75,1]}}
 {id=Mixer_1, children [WavePlayer_1,null], properties {inputVolume [1,1]}}
 {id=Modulator_1, children [WavePlayer_6]}
 {id=WavePlayer_1, properties {soundWaveAssetPtr=&/Cash_PickUp_Tone_High}}
 {id=WavePlayer_6, properties {soundWaveAssetPtr=&/Cash_Pickup_High}}
]
properties {
 attenuationSettings=&Fort_Weapon_Pickup_Att
 maxDistance=3500
 soundClassObject=&Fort_Loot_Collect
 volumeMultiplier=0.25
 attenuationOverrides {falloffDistance=350, nonSpatializedRadiusStart=200, stereoSpread=512, attenuationShapeExtents {x=150}}
 concurrencyOverrides {maxCount=2, resolutionRule=StopOldest}
 concurrencySet [&/CONC_Wad_Pickup]
}
