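# Sound cue played for picking up a small wad item, built from two low cash pickup sounds.
# A Mixer layers the cash pickup wave, run through a default Modulator, with a slightly quieter tone wave that starts after a short fixed delay.
# It uses weapon pickup attenuation and wad pickup concurrency, but borrows a first-person ranged-fire sound class and gunfire property sheet.
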SoundCue /Game/Athena/Sounds/Items/Wads/Wad_PickUp_Small_Cue
root=Mixer_0
nodes [
 {id=Delay_0, children [WavePlayer_1], properties {delayMax=0.1, delayMin=0.1}}
 {id=Mixer_0, children [Modulator_1,Delay_0], properties {inputVolume [1,0.9]}}
 {id=Modulator_1, children [WavePlayer_6]}
 {id=WavePlayer_1, properties {soundWaveAssetPtr=&/Cash_PickUp_Tone_Low}}
 {id=WavePlayer_6, properties {soundWaveAssetPtr=&/Cash_Pickup_Low}}
]
properties {
 attenuationSettings=&Fort_Weapon_Pickup_Att
 audioPropertiesSheet=/FortMetaSoundWeapon/Items/Shared/PropertySheets/APS_1P_Weapon_Gunfire
 maxDistance=3500
 soundClassObject=&Fort_WeaponRanged_Fire_1P_PSM
 volumeMultiplier=0.3
 assetUserData [&self:AudioAssetUserData_0]
 attenuationOverrides {falloffDistance=350, nonSpatializedRadiusStart=200, stereoSpread=512, attenuationShapeExtents {x=150}}
 concurrencyOverrides {maxCount=2, resolutionRule=StopOldest}
 concurrencySet [&/CONC_Wad_Pickup]
}
